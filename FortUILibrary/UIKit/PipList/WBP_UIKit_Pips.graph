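# UIKit pip indicator widget: a row of pips between previous and next arrow buttons, showing and changing the active index.
# SetupPips rebuilds the entries in a DynamicEntryBox and activates the first.
# Button clicks, also triggered by configurable input actions, step the index with modulo wrap-around into SetActivePip, which swaps the active entry and fires OnActivePipChanged.
# On input method changes, the arrows' main panels and backgrounds collapse for gamepad.
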WidgetBlueprint /FortUILibrary/UIKit/PipList/WBP_UIKit_Pips
graphs [
 EventGraph {nodes [
   "On Clicked (Button_Left)"!() -> (OutputDelegate, then, Button:Common Button Base Object Reference) {is=title K2Node_ComponentBoundEvent}
   n1 = Get PipCount() -> (PipCount:int=0)
   n2 = Get PipIndex() -> (PipIndex:int=0)
   > n3 = Set PipIndex(PipIndex:int="% (Integer)"(A=int + int!(A=@n2, B=@n1, C:int=-1) -> (ReturnValue), B=Get PipCount() -> (PipCount=0)) -> (ReturnValue=0)) -> (Output_Get:int=0)
   SetActivePip!(execute=[@n3.then, @n5.then], self, Index:int=Get PipIndex() -> (PipIndex=0)) -> (then)
   "On Clicked (Button_Right)"!() -> (OutputDelegate, then, Button:Common Button Base Object Reference) {is=title K2Node_ComponentBoundEvent}
   n4 = Get PipIndex() -> (PipIndex=0)
   > n5 = Set PipIndex(PipIndex="% (Integer)"(A=int + int!(A=@n4, B) -> (ReturnValue), B=Get PipCount() -> (PipCount=0)) -> (ReturnValue=0)) -> (Output_Get=0)
   Event PreConstruct()
   > SetTriggeringInputAction!(
    execute
    self:Common Button Base Object Reference=Get Button_Left() -> (Button_Left:WBP UIKit Button Generic Object Reference)
    InputActionRow:"Data Table Row Handle Structure (by ref)"=Get LeftButtonInputAction() -> (LeftButtonInputAction:Data Table Row Handle Structure)
   ) -> (then)
   > SetTriggeringInputAction(
    self=Get Button_Right() -> (Button_Right:WBP UIKit Button Generic Object Reference)
    InputActionRow=Get RightButtonInputAction() -> (RightButtonInputAction:Data Table Row Handle Structure)
   )
   > UpdateArrowButtonPresentation!(execute, self, InputPin:ECommonInputType Enum=GetCurrentInputType(self=Get CommonInputSubsystem())) -> (then)
   > Bind Event to On Input Method Changed(self=Get CommonInputSubsystem(), Delegate=Create Event())
  ]}
 OnActivePipChanged {nodes [OnActivePipChanged!() -> (then, NewIndex:int) {is=title K2Node_FunctionEntry}]}
 SetActivePip {nodes [
   n1 = SetActivePip!() -> (then, Index:int) {is=title K2Node_FunctionEntry}
   > Branch(Condition=Is Valid Index!(
     TargetArray:Array of WBP UIKit Pip Entry Object References=Get PipsArray() -> (PipsArray:Array of WBP UIKit Pip Entry Object References)
     IndexToTest:int=@n1
    ) -> (ReturnValue:bool))
   > n2 = Get!(execute:Exec) -> (then:Exec, else:Exec, Current Pip Entry:WBP UIKit Pip Entry Object Reference)
   > n3 = SetActive!(execute, self:WBP UIKit Pip Entry Object Reference=@n2, Active:bool=false, IsImmediate:bool=false) -> (then)
   n4 = Set Current Pip Entry(
    execute=[@n3.then, @n2.else]
    Current Pip Entry:WBP UIKit Pip Entry Object Reference="Get (a copy)"(Array:Array of WBP UIKit Pip Entry Object References=Get PipsArray(), Dimension 1=@n1) -> (Output:WBP UIKit Pip Entry Object Reference)
   ) -> (Output_Get:WBP UIKit Pip Entry Object Reference)
   > SetActive(self=@n4, Active=true, IsImmediate=false)
   > n5 = Set PipIndex(PipIndex=@n1) -> (Output_Get=0)
   > Call On Active Pip Changed!(execute:Exec, self:WBP UIKit Pips Object Reference, NewIndex:int=@n5) -> (then:Exec) {is=title K2Node_CallDelegate}
  ]}
 SetupPips {nodes [
   n1 = SetupPips!() -> (then, Count:int) {is=title K2Node_FunctionEntry}
   > Set PipCount(PipCount:int=@n1) -> (Output_Get:int=0)
   > Clear(TargetArray:Array of WBP UIKit Pip Entry Object References=Get PipsArray())
   > Reset!(execute, self:Dynamic Entry Box Object Reference=Get PipBox() -> (PipBox:Dynamic Entry Box Object Reference), bDeleteWidgets:bool=true) -> (then)
   > n2 = For Loop(FirstIndex=1, LastIndex=Get PipCount() -> (PipCount=0))
   SetActivePip(execute=@n2.Completed, Index=0) {is=title K2Node_CallFunction}
   n3 = Create Entry!(execute=@n2.LoopBody, self:Dynamic Entry Box Object Reference=Get PipBox()) -> (then, ReturnValue:User Widget Object Reference)
   > n4 = Cast To WBP_UIKit_PipEntry!(execute, Object=@n3) -> (then, CastFailed, AsWBP UIKit Pip Entry:WBP UIKit Pip Entry Object Reference) {is=title K2Node_DynamicCast}
   > SetActive(self=@n4, Active=false, IsImmediate=true)
   > Add(TargetArray:Array of WBP UIKit Pip Entry Object References=Get PipsArray(), NewItem:"WBP UIKit Pip Entry Object Reference (by ref)"=@n4) {is=title K2Node_CallArrayFunction}
  ]}
 UpdateArrowButtonPresentation {nodes [
   n1 = UpdateArrowButtonPresentation!() -> (then, InputPin:ECommonInputType Enum) {is=title K2Node_FunctionEntry}
   n2 = Get Button_Left()
   n3 = Get Button_Right()
   > SetVisibility(
    self=[
     Get Main!(self:WBP UIKit Button Generic Base Object Reference=@n2) -> (Main:Grid Panel Object Reference)
     Get Main(self=@n3)
     Get Block_Background!(self:WBP UIKit Button Generic Base Object Reference=@n3) -> (Block_Background:WBP UIKit Block Background Base Object Reference)
     Get Block_Background(self=@n2)
    ]
    InVisibility=Select!(
     MouseAndKeyboard:ESlateVisibility Enum=SelfHitTestInvisible
     Gamepad:ESlateVisibility Enum=Collapsed
     Touch:ESlateVisibility Enum=SelfHitTestInvisible
     Count:ESlateVisibility Enum=SelfHitTestInvisible
     Index:ECommonInputType Enum=@n1
    ) -> (ReturnValue)
   )
  ]}
]
variables [
 Current Pip Entry {type=WBP_UIKit_PipEntry_C, default=null}
 LeftButtonInputAction {type=DataTableRowHandle, default {dataTable=&InputActionDataTable, rowName=MainTabLeft}}
 OnActivePipChanged {}
 PipCount {type=integer, default=4}
 PipIndex {type=integer}
 PipsArray {type="array<WBP_UIKit_PipEntry_C>", default []}
 RightButtonInputAction {type=DataTableRowHandle, default {dataTable=&InputActionDataTable, rowName=MainTabRight}}
]
widgets {
 info {inheritedWidgetCount=0, parentClass=&UserWidget, rootWidgetClass=&GridPanel, widgetCount=4}
 widgets [
  GridPanel_0 {widgetClassPath=&GridPanel, widget=GridPanel_0}
  Button_Left {
   widgetClassPath=&WBP_UIKit_Button_Generic
   bIsVariable=true
   parent=GridPanel_0
   widget=Button_Left
   properties {
    bHideInputActionWithKeyboard=true
    bIsFocusable=false
    iconMaterial=/FortUILibrary/Foundation/Buttons/Materials/Arrows/MI_UIKit_Icon_Button_Previous
    minHeight=48
    minHeightOverride=48
    minWidthOverride=48
    mobileMinHeight=88
    mobileMinWidth=88
    outlineMaterial=/FortUILibrary/UIKit/Buttons/MaterialInstances/TileOutline/MI_UIKit_Outline
    promptHorizontalAlignment=HAlign_Center
    useTextBlock=false
    buttonPadding=<ltrb 0 0 0 0>
    promptNudge=<0 0>
    triggeringInputAction {dataTable=&InputActionDataTable, rowName=MainTabLeft}
   }
   slot {class=GridSlot, properties {padding=<ltrb 0 0 16 0>}}
  }
  PipBox {
   widgetClassPath=&DynamicEntryBox
   bIsVariable=true
   parent=GridPanel_0
   widget=PipBox
   properties {
    entryHorizontalAlignment=HAlign_Left
    entryVerticalAlignment=VAlign_Center
    entryWidgetClass=&/WBP_UIKit_PipEntry.WBP_UIKit_PipEntry_C
    numDesignerPreviewEntries=5
    visibility=HitTestInvisible
    entrySpacing=<4 0>
    radialBoxSettings {angleBetweenItems=0, bClockwiseOrder=false, bDistributeItemsEvenly=false, marginSize=0, sectorCentralAngle=360, startingAngle=0}
   }
   slot {class=GridSlot, properties {column=1}}
  }
  Button_Right {
   widgetClassPath=&WBP_UIKit_Button_Generic
   bIsVariable=true
   parent=GridPanel_0
   widget=Button_Right
   properties {
    bHideInputActionWithKeyboard=true
    bIsFocusable=false
    iconMaterial=/FortUILibrary/Foundation/Buttons/Materials/Arrows/MI_UIKit_Icon_Button_Next
    minHeight=0
    minHeightOverride=48
    minWidthOverride=48
    mobileMinHeight=88
    mobileMinWidth=88
    outlineMaterial=&MI_UIKit_Outline
    promptHorizontalAlignment=HAlign_Center
    useTextBlock=false
    buttonPadding=<ltrb 0 0 0 0>
    promptNudge=<0 0>
    triggeringInputAction {dataTable=&InputActionDataTable, rowName=MainTabRight}
   }
   slot {class=GridSlot, properties {column=2, padding=<ltrb 16 0 0 0>}}
  }
 ]
}
