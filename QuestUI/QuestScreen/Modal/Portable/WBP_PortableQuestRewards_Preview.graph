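# Widget that previews a quest reward in the portable quest rewards modal, stacking a lazy image, content slot, type text and name text.
# Helper functions return the first reward of the indexed quest view model and prefer an eligible large preview over a small one.
# SetPreviewImage loads the lazy image and shows it, or hides it without a valid asset.
# On Construct, compact UI gets adjusted content slot padding.
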WidgetBlueprint /QuestUI/QuestScreen/Modal/Portable/WBP_PortableQuestRewards_Preview
graphs [
 EventGraph {nodes [
   Event Construct()
   > SetMobileStyle() {is=title K2Node_CallFunction}
  ]}
 GetPreviewImage {nodes [
   n1 = GetPreviewImage!() -> (then, SmallPreview:Soft Object Reference, SmallPreviewEligible:bool=~, LargePreview:Soft Object Reference, LargePreviewElligible:bool=~) {is=title K2Node_FunctionEntry}
   > Return Node!(
    execute
    ReturnValue:Soft Object Reference=Select(
     Option 0:Soft Object Reference=Select(Option 1=@n1, Index=@n1.SmallPreviewEligible)
     Option 1:Soft Object Reference=@n1.LargePreview
     Index=@n1.LargePreviewElligible
    ) -> (ReturnValue:Soft Object Reference)
   )
  ]}
 GetRewardFromIndex {nodes [
   n1 = GetRewardFromIndex!() -> (then, InIndex:int, InRewardArray:Array of Quest UI Base Content View Model Object References) {is=title K2Node_FunctionEntry}
   > n2 = Branch(Condition=Is Valid Index!(TargetArray:Array of Quest UI Base Content View Model Object References=@n1.InRewardArray, IndexToTest:int=@n1) -> (ReturnValue:bool))
   > n3 = Cast To QuestUIQuestVM!(
    execute
    Object="Get (a copy)"(
     Array:Array of Quest UI Base Content View Model Object References=Get InRewardArray() -> (InRewardArray:Array of Quest UI Base Content View Model Object References)
     Dimension 1=Get InIndex() -> (InIndex:int=0)
    ) -> (Output:Quest UI Base Content View Model Object Reference)
   ) -> (then, CastFailed, AsQuest UI Quest View Model:Quest UI Quest View Model Object Reference) {is=title K2Node_DynamicCast}
   n4 = Get RewardVMs!(self:Quest UI Quest View Model Object Reference=@n3) -> (RewardVMs:Array of Quest UI Reward VM Object References)
   > n5 = Branch(Condition=Is Valid Index(TargetArray:Array of Quest UI Reward VM Object References=@n4))
   > Return Node!(
    execute
    OutReward:Quest UI Reward VM Object Reference="Get (a copy)"(Array:Array of Quest UI Reward VM Object References=@n4) -> (Output:Quest UI Reward VM Object Reference)
   )
   Return Node(execute=@n5.else)
   Return Node(execute=@n3.CastFailed)
   Return Node(execute=@n2.else)
  ]}
 SetMobileStyle {nodes [
   SetMobileStyle() {is=title K2Node_FunctionEntry}
   > Branch(Condition=IsUsingCompactUI())
   > SetGridSlotPadding!(
    execute
    Widget:Widget Object Reference=Get Slot_Content() -> (Slot_Content:Named Slot Object Reference)
    Padding_Left:float=0
    Padding_Top:float=20
    Padding_Right:float=0
    Padding_Bottom:float=0
   ) -> (then)
  ]}
 SetPreviewImage {nodes [
   SetPreviewImage!() -> (then, LazyObject:Soft Object Reference) {is=title K2Node_FunctionEntry}
   > n1 = Get!(execute:Exec) -> (then:Exec, else:Exec, LazyObject:Soft Object Reference)
   n2 = Get LazyImage_Preview() -> (LazyImage_Preview:Fort Lazy Image Object Reference)
   n3 = Get LazyImage_Preview()
   > SetBrushFromLazyDisplayAsset!(
    execute
    self:Common Lazy Image Object Reference=[@n2, @n3]
    LazyObject:"Soft Object Reference (by ref)"=Get LazyObject() -> (LazyObject:Soft Object Reference)
    bMatchTextureSize:bool=false
   ) -> (then)
   > SetVisibility(self=@n2, InVisibility=SelfHitTestInvisible)
   SetVisibility(execute=@n1.else, self=@n3, InVisibility=Hidden)
  ]}
]
variables []
widgets {
 info {inheritedWidgetCount=0, parentClass=&UserWidget, rootWidgetClass=&VerticalBox, widgetCount=7}
 widgets [
  VerticalBox_17 {widgetClassPath=&VerticalBox, widget=VerticalBox_17}
  LazyImage_Preview {
   widgetClassPath=&FortLazyImage
   bIsVariable=true
   parent=VerticalBox_17
   widget=LazyImage_Preview
   properties {brush {
     bIsDynamicallyLoaded=false
     drawAs=Image
     imageType=NoImage
     mirroring=NoMirror
     resourceName=None
     resourceObject=null
     tiling=NoTile
     imageSize=<300 300>
     margin=<ltrb 0 0 0 0>
     outlineSettings {
      bUseBrushTransparency=false
      roundingType=HalfHeightRadius
      width=0
      color {colorUseRule=UseColor_Specified, specifiedColor=<rgba 0 0 0 0>}
      cornerRadii=<0 0 0 0>
     }
     tintColor {colorUseRule=UseColor_Specified, specifiedColor=<rgba 1 1 1 1>}
     uVRegion {bIsValid=false, max=<0 0>, min=<0 0>}
    }}
   slot {class=VerticalBoxSlot, properties {horizontalAlignment=HAlign_Center, verticalAlignment=VAlign_Center}}
  }
  Slot_Content {
   widgetClassPath=&NamedSlot
   bIsVariable=true
   parent=VerticalBox_17
   widget=Slot_Content
   slot {class=VerticalBoxSlot, properties {horizontalAlignment=HAlign_Center, verticalAlignment=VAlign_Top, padding=<ltrb 0 36 0 0>}}
  }
  Spacer_40 {
   widgetClassPath=&Spacer
   bIsVariable=true
   parent=VerticalBox_17
   widget=Spacer_40
   slot {class=VerticalBoxSlot, properties {size {sizeRule=Fill, value=1}}}
  }
  Text_Type {
   widgetClassPath=&CommonTextBlock
   parent=VerticalBox_17
   widget=Text_Type
   properties {
    text="[PH] Outfit"
    textTransformPolicy=ToUpper
    colorAndOpacity {colorUseRule=UseColor_Specified, specifiedColor=<rgba 0.5776 0.5776 0.5776 1>}
    font {
     bForceMonospaced=false
     bMaterialIsStencil=false
     fontMaterial=null
     fontObject=&HeadingNow
     letterSpacing=0
     monospacedWidth=1
     size=18
     skewAmount=0
     typefaceFontName=75 Medium Italic
     outlineSettings {
      bApplyOutlineToDropShadows=false
      bMiteredCorners=false
      bSeparateFillAlpha=false
      outlineMaterial=null
      outlineSize=0
      outlineColor=<rgba 0 0 0 1>
     }
    }
    shadowOffset=<0 0>
   }
   slot {class=VerticalBoxSlot, properties {verticalAlignment=VAlign_Bottom}}
  }
  SB_Title {
   widgetClassPath=&SizeBox
   bIsVariable=true
   parent=VerticalBox_17
   widget=SB_Title
   properties {bOverride_HeightOverride=true, heightOverride=94}
   slot {class=VerticalBoxSlot, properties {verticalAlignment=VAlign_Bottom}}
  }
  Text_Name {
   widgetClassPath=&CommonTextBlock
   parent=SB_Title
   widget=Text_Name
   properties {
    autoWrapText=true
    clipping=ClipToBounds
    text="[PH] Long Item Name multiple lines"
    textOverflowPolicy=MultilineEllipsis
    textTransformPolicy=ToUpper
    font {
     bForceMonospaced=false
     bMaterialIsStencil=false
     fontMaterial=null
     fontObject=&HeadingNow
     letterSpacing=0
     monospacedWidth=1
     size=25.5
     skewAmount=0
     typefaceFontName=86 Bold Italic
     outlineSettings {
      bApplyOutlineToDropShadows=false
      bMiteredCorners=false
      bSeparateFillAlpha=false
      outlineMaterial=null
      outlineSize=0
      outlineColor=<rgba 0 0 0 1>
     }
    }
    shadowOffset=<0 0>
   }
   slot {class=SizeBoxSlot}
  }
 ]
}
