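# Selectable UIKit modular button for one numbered reward entry in the portable quest rewards UI, showing item preview, quantity, progress bar and owned/locked indicators.
# Focus or click sets the punch card view model's SelectedIndex from the goal reward's index.
# Setter functions drive preview and progress materials, completion colors and indicators; account-resource quantities abbreviate thousands as "k".
# A registered outline block shows hover as focus and is visible only for mouse and keyboard input.
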WidgetBlueprint /QuestUI/QuestScreen/Modal/Portable/WBP_PortableQuestRewards_Entry
graphs [
 EventGraph {nodes [
   Event Construct()
   > RegisterModularBlock!(
    execute
    self:UIKit Modular Button Object Reference
    Block:UIKit Block Object Reference=Get WBP_UIKit_Block_Outline() -> (WBP_UIKit_Block_Outline:WBP UIKit Block Outline Object Reference)
   ) -> (then)
   > SetBlocksTiming!(
    execute
    self:UIKit Modular Button Object Reference
    Timing_Hovering:float=0
    Timing_Unhovering:float=0
    Timing_Focusing:float=0.1
    Timing_Unfocusing:float=0
    Timing_Pressing:float=0.1
    Timing_Releasing:float=0
    Timing_Disabling:float=0
    Timing_Enabling:float=0
    Timing_Selecting:float=0
    Timing_DeselectingIdle:float=0
    Timing_DeselectingFocused:float=0
   ) -> (then)
   n1 = Get CommonInputSubsystem()
   > Bind Event to On Input Method Changed(self=@n1, Delegate=Create Event())
   > UpdateOutlineVisibility!(execute, self, Index:ECommonInputType Enum=GetCurrentInputType(self=@n1)) -> (then)
   n2 = Event On Focused() {is=title K2Node_Event}
   Set with Broadcast SelectedIndex!(
    execute=[@n2.then, @n3.then]
    SelectedIndex:int=Get Index!(self:Quest UIGoal Reward VM Object Reference=Get QuestUIGoalRewardVM() -> (QuestUIGoalRewardVM:Quest UIGoal Reward VM Object Reference)) -> (Index:int=0)
    self:Quest UIHorizontal Punch Card Object Reference=Get QuestUIHorizontalPunchCard() -> (QuestUIHorizontalPunchCard:Quest UIHorizontal Punch Card Object Reference)
   ) -> (then, Output_Get:int=0)
   n3 = Event On Clicked()
   Event On Hovered()
   > RequestTransitionOnBlocks!(execute, self:UIKit Modular Button Object Reference, State:EUIKitBlockVisualState Enum=Focused) -> (then)
   Event On Unhovered()
   > RequestTransitionOnBlocks(State=Unfocused)
  ]}
 GetQuantityText {nodes [
   GetQuantityText!() -> (then, InQuantity:int, InRewardType:EFortItemType Enum=WorldItem) {is=title K2Node_FunctionEntry}
   n1 = "Equal (Enum)"(A:EFortItemType Enum=Get InRewardType() -> (InRewardType:EFortItemType Enum=WorldItem), B:EFortItemType Enum=AccountResource)
   n2 = Get InQuantity() -> (InQuantity:int=0)
   n3 = "Not Equal (Integer)"(A=@n2, B=1)
   n4 = "To Text (Integer)"(Value=@n2)
   n5 = int / int(A=@n2, B=1000)
   n6 = Format Text!(Format="LOCTEXT(\"{Amount}k\")", Amount:int=@n5) -> (Result)
   n7 = "integer >= integer"(A=@n5, B=1)
   > Return Node!(
    execute
    NewParam:Text=Select(Option 0:Text, Option 1:Text=Select(Option 0=@n4, Option 1=@n6, Index=@n7), Index=AND Boolean(A=@n3, B=@n1)) -> (ReturnValue:Text)
   )
  ]}
 SetDisplayingFirstReward {nodes [
   n1 = SetDisplayingFirstReward!() -> (then, InRewards:Array of Quest UI Reward VM Object References) {is=title K2Node_FunctionEntry}
   > Branch(Condition=Is Valid Index!(TargetArray:Array of Quest UI Reward VM Object References=@n1, IndexToTest:int) -> (ReturnValue:bool))
   > Set QuestUIRewardVM(QuestUIRewardVM:Quest UI Reward VM Object Reference="Get (a copy)"(Array:Array of Quest UI Reward VM Object References=@n1) -> (Output:Quest UI Reward VM Object Reference)) -> (Output_Get:Quest UI Reward VM Object Reference)
  ]}
 SetIndex {nodes [
   n1 = SetIndex!() -> (then, InIndex:int) {is=title K2Node_FunctionEntry}
   > "SetText (Text)"(self=Get Text_Index() -> (Text_Index:Common Text Object Reference), InText="To Text (Integer)"(Value=int + int(A=@n1)))
   n2 = Get InIndex() -> (InIndex:int=0)
   > SetVisibility(
    self=Get ProgressBar() -> (ProgressBar:Image Object Reference)
    InVisibility=Select(Option 0:ESlateVisibility Enum=HitTestInvisible, Option 1:ESlateVisibility Enum=Collapsed, Index="Equal (Integer)"(A=@n2)) -> (ReturnValue:ESlateVisibility Enum)
   )
  ]}
 SetPreviewImage {nodes [
   SetPreviewImage!() -> (then, LazyObject:Soft Object Reference) {is=title K2Node_FunctionEntry}
   > n1 = Get!(execute:Exec) -> (then:Exec, else:Exec, LazyObject:Soft Object Reference)
   n2 = Get LazyImage_Display() -> (LazyImage_Display:Fort Lazy Image Object Reference)
   > SetMaterialTextureParamName!(execute, self:Common Lazy Image Object Reference=@n2, TextureParamName:Name=Item) -> (then)
   > SetBrushFromLazyDisplayAsset!(
    execute
    self:Common Lazy Image Object Reference=@n2
    LazyObject:"Soft Object Reference (by ref)"=Get LazyObject() -> (LazyObject:Soft Object Reference)
    bMatchTextureSize:bool=false
   ) -> (then)
   SetVisibility(execute=@n1.else, self=Get LazyImage_Display(), InVisibility=Collapsed)
  ]}
 SetPreviewScaleAndOffset {nodes [
   SetPreviewScaleAndOffset!() -> (then, InRewardType:EFortItemType Enum=WorldItem) {is=title K2Node_FunctionEntry}
   > n1 = GetDynamicMaterial(self=Get LazyImage_Display())
   > Set MID(MID:Material Instance Dynamic Object Reference=@n1) -> (Output_Get:Material Instance Dynamic Object Reference)
   > n2 = Cast To QuestUIRewardCinematicVM!(execute, Object=Get QuestUIRewardVM() -> (QuestUIRewardVM:Quest UI Reward VM Object Reference)) -> (then, CastFailed, AsQuest UI Reward Cinematic VM:Quest UI Reward Cinematic VM Object Reference) {is=title K2Node_DynamicCast}
   n3 = Get MID() -> (MID:Material Instance Dynamic Object Reference)
   > SetScalarParameterValue(self=@n3, ParameterName=ScaleFrom, Value=1.3)
   > SetScalarParameterValue(self=@n3, ParameterName=ItemOffset_Y, Value=0)
   n4 = Branch(execute=@n2.CastFailed, Condition="Equal (Enum)"(A=Get InRewardType() -> (InRewardType=WorldItem), B=AccountResource))
   n5 = Get MID()
   > SetScalarParameterValue(self=@n5, ParameterName=ScaleFrom, Value=0.6)
   > SetScalarParameterValue(self=@n5, ParameterName=ItemOffset_Y, Value=-0.1)
   n6 = Get MID()
   SetScalarParameterValue(execute=@n4.else, self=@n6, ParameterName=ScaleFrom, Value=1)
   > SetScalarParameterValue(self=@n6, ParameterName=ItemOffset_Y, Value=0)
  ]}
 SetProgress {nodes [
   n1 = SetProgress!() -> (then, InProgress:double=~) {is=title K2Node_FunctionEntry}
   > Set IsProgressComplete(IsProgressComplete:bool="float >= float"(A=@n1, B=1)) -> (Output_Get:bool)
   > n2 = Sequence!(execute) -> (then_0, then_1, then_2:Exec, then_3:Exec)
   > n3 = GetDynamicMaterial(self=Get ProgressBar())
   > n4 = Set ProgressMaterial(ProgressMaterial:Material Instance Dynamic Object Reference=@n3) -> (Output_Get:Material Instance Dynamic Object Reference)
   > SetScalarParameterValue(self=@n4, ParameterName=AnimateProgress, Value=Get InProgress() -> (InProgress:double))
   n5 = Get IsProgressComplete() -> (IsProgressComplete:bool)
   SetColorAndOpacity(
    execute=@n2.then_2
    self:Image Object Reference=Get IndexBackground() -> (IndexBackground:Image Object Reference)
    InColorAndOpacity:Linear Color Structure=Select(
     Option 0:Linear Color Structure=Get EmptyProgressColor() -> (EmptyProgressColor:Linear Color Structure)
     Option 1:Linear Color Structure=Get CompleteProgressColor() -> (CompleteProgressColor:Linear Color Structure)
     Index=@n5
    ) -> (ReturnValue:Linear Color Structure)
   )
   > SetColorAndOpacity(
    self:Text Object Reference=Get Text_Index()
    InColorAndOpacity:Slate Color Structure=Select!(
     Option 0_SpecifiedColor:Linear Color Structure=(R=1,G=1,B=1,A=1)
     Option 0_ColorUseRule:ESlateColorStylingMode Enum
     Option 1_SpecifiedColor:Linear Color Structure=(R=0.0048,G=0.004,B=0.0116,A=1)
     Option 1_ColorUseRule:ESlateColorStylingMode Enum
     Index=@n5
    ) -> (ReturnValue:Slate Color Structure)
   )
   n6 = Select!(
    Option 0:Linear Color Structure=Get InProgressColor() -> (InProgressColor:Linear Color Structure)
    Option 1:Linear Color Structure=Get CompleteProgressColor()
    Index=Get IsProgressComplete()
   ) -> (ReturnValue:Linear Color Structure)
   n7 = Get ProgressMaterial() -> (ProgressMaterial:Material Instance Dynamic Object Reference)
   SetVectorParameterValue(execute=@n2.then_1, self=@n7, ParameterName=ColorA, Value=@n6)
   > SetVectorParameterValue(self=@n7, ParameterName=ColorB, Value=@n6)
   > SetVectorParameterValue(self=@n7, ParameterName=ColorC, Value=@n6)
   n8 = Get IsProgressComplete()
   Set Visibility(
    execute=@n2.then_3
    Visibility=Select(Option 0:ESlateVisibility Enum=Collapsed, Option 1:ESlateVisibility Enum=SelfHitTestInvisible, Index=@n8) -> (ReturnValue:ESlateVisibility Enum)
    self=Get Indicator_Owned() -> (Indicator_Owned:WBP UIKit Status Indicator Object Reference)
   )
   > Set Visibility(
    Visibility=Select(Option 0=SelfHitTestInvisible, Option 1=Collapsed, Index=@n8)
    self=Get Indicator_Locked() -> (Indicator_Locked:WBP UIKit Status Indicator Object Reference)
   )
  ]}
 SetSelected {nodes [
   n1 = SetSelected!() -> (then, InSelected:bool=~) {is=title K2Node_FunctionEntry}
   > n2 = GetDynamicMaterial(self=Get LazyImage_Display())
   > SetScalarParameterValue(self=@n2, ParameterName=UseOutline, Value="To Float (Boolean)"(InBool=@n1))
  ]}
 UpdateOutlineVisibility {nodes [
   n1 = UpdateOutlineVisibility!() -> (then, Index:ECommonInputType Enum) {is=title K2Node_FunctionEntry}
   > Set Visibility(
    Visibility=Select!(
     MouseAndKeyboard:ESlateVisibility Enum=SelfHitTestInvisible
     Gamepad:ESlateVisibility Enum=Collapsed
     Touch:ESlateVisibility Enum=Collapsed
     Count:ESlateVisibility Enum=Collapsed
     Index:ECommonInputType Enum=@n1
    ) -> (ReturnValue)
    self=Get WBP_UIKit_Block_Outline()
   )
  ]}
]
variables [
 CompleteProgressColor {type=LinearColor, default=<rgba 0.0331 0.9911 0.2195 1>}
 EmptyProgressColor {type=LinearColor, default=<rgba 1 1 1 0.5>}
 InProgressColor {type=LinearColor, default=<rgba 0.9301 1 0.0103 1>}
]
widgets {
 info {inheritedWidgetCount=1, parentClass=/Script/FortniteUI.UIKitModularButton, rootWidgetClass=&GridPanel, widgetCount=13}
 widgets [
  GridPanel_0 {widgetClassPath=&GridPanel, widget=GridPanel_0, properties {bOverride_Cursor=true, rowFill [1,0]}}
  LazyImage_Display {
   widgetClassPath=&FortLazyImage
   bIsVariable=true
   parent=GridPanel_0
   widget=LazyImage_Display
   properties {brush {
     bIsDynamicallyLoaded=false
     drawAs=Image
     imageType=FullColor
     mirroring=NoMirror
     resourceName=None
     resourceObject=&/MI_UI_PortableQuestRewards_EntryBackground
     tiling=NoTile
     imageSize=<128 142>
     margin=<ltrb 0 0 0 0>
     outlineSettings {
      bUseBrushTransparency=false
      roundingType=HalfHeightRadius
      width=0
      color {colorUseRule=UseColor_Specified, specifiedColor=<rgba 0 0 0 0>}
      cornerRadii=<0 0 0 0>
     }
     tintColor {colorUseRule=UseColor_Specified, specifiedColor=<rgba 1 1 1 1>}
     uVRegion {bIsValid=false, max=<0 0>, min=<0 0>}
    }}
   slot {class=GridSlot, properties {padding=<ltrb 0 0 0 16>}}
  }
  Indicator_Owned {
   widgetClassPath=/FortUILibrary/UIKit/StatusIndicator/WBP_UIKit_StatusIndicator
   bIsVariable=true
   parent=GridPanel_0
   widget=Indicator_Owned
   properties {icon=/FortUILibrary/Textures/Icons/T_UI_Icon_UIKit_Check, iconSize=Compact}
   slot {class=GridSlot, properties {horizontalAlignment=HAlign_Right, verticalAlignment=VAlign_Top, padding=<ltrb 8 8 8 8>}}
  }
  Indicator_Locked {
   widgetClassPath=&WBP_UIKit_StatusIndicator
   bIsVariable=true
   parent=GridPanel_0
   widget=Indicator_Locked
   properties {type=Utility, icon=/FortUILibrary/Textures/Icons/ToBeDeprecated/T_UI_Icon_Global_Locked, iconSize=Compact}
   slot {class=GridSlot, properties {horizontalAlignment=HAlign_Right, verticalAlignment=VAlign_Top, padding=<ltrb 8 8 8 8>}}
  }
  Text_Quantity {
   widgetClassPath=&CommonTextBlock
   bIsVariable=true
   parent=GridPanel_0
   widget=Text_Quantity
   properties {
    bAutoCollapseWithEmptyText=true
    justification=Center
    scrollStyle=&TextScrollStyle-Base_C
    text=100k
    font {
     bForceMonospaced=false
     bMaterialIsStencil=false
     fontMaterial=null
     fontObject=&HeadingNow
     letterSpacing=0
     monospacedWidth=1
     size=18
     skewAmount=0
     typefaceFontName=75 Medium
     outlineSettings {
      bApplyOutlineToDropShadows=false
      bMiteredCorners=false
      bSeparateFillAlpha=false
      outlineMaterial=null
      outlineSize=0
      outlineColor=<rgba 0 0 0 1>
     }
    }
    shadowOffset=<0 0>
   }
   slot {class=GridSlot, properties {verticalAlignment=VAlign_Bottom, padding=<ltrb 0 0 0 38>}}
  }
  FortVisualAttachment_0 {
   widgetClassPath=&FortVisualAttachment
   parent=GridPanel_0
   widget=FortVisualAttachment_0
   properties {contentAnchor=<0.5 0.5>}
   slot {class=GridSlot, properties {horizontalAlignment=HAlign_Left, row=1, verticalAlignment=VAlign_Center, nudge=<-16 0>}}
  }
  ProgressBar {
   widgetClassPath=&Image
   bIsVariable=true
   parent=FortVisualAttachment_0
   widget=ProgressBar
   properties {
    renderOpacity=0.8
    visibility=SelfHitTestInvisible
    brush {
     bIsDynamicallyLoaded=false
     drawAs=Image
     imageType=FullColor
     mirroring=NoMirror
     resourceName=None
     resourceObject=&/MI_UI_PortableQuestRewards_Progress
     tiling=NoTile
     imageSize=<100 16>
     margin=<ltrb 0 0 0 0>
     outlineSettings {
      bUseBrushTransparency=false
      roundingType=HalfHeightRadius
      width=0
      color {colorUseRule=UseColor_Specified, specifiedColor=<rgba 0 0 0 0>}
      cornerRadii=<0 0 0 0>
     }
     tintColor {colorUseRule=UseColor_Specified, specifiedColor=<rgba 1 1 1 1>}
     uVRegion {bIsValid=false, max=<0 0>, min=<0 0>}
    }
   }
   slot {class=SizeBoxSlot}
  }
  IndexBackground {
   widgetClassPath=&Image
   bIsVariable=true
   parent=GridPanel_0
   widget=IndexBackground
   properties {
    renderOpacity=0.8
    visibility=SelfHitTestInvisible
    brush {
     bIsDynamicallyLoaded=false
     drawAs=RoundedBox
     imageType=NoImage
     mirroring=NoMirror
     resourceName=None
     resourceObject=null
     tiling=NoTile
     imageSize=<36 36>
     margin=<ltrb 0 0 0 0>
     outlineSettings {
      bUseBrushTransparency=false
      roundingType=HalfHeightRadius
      width=0
      color {colorUseRule=UseColor_Specified, specifiedColor=<rgba 0 0 0 0>}
      cornerRadii=<0 0 0 0>
     }
     tintColor {colorUseRule=UseColor_Specified, specifiedColor=<rgba 1 1 1 1>}
     uVRegion {bIsValid=false, max=<0 0>, min=<0 0>}
    }
    colorAndOpacity=<rgba 0.0513 0.0513 0.0513 0.502>
   }
   slot {class=GridSlot, properties {horizontalAlignment=HAlign_Center, row=1}}
  }
  SizeBox_0 {
   widgetClassPath=&SizeBox
   parent=GridPanel_0
   widget=SizeBox_0
   properties {bOverride_WidthOverride=true, widthOverride=36}
   slot {class=GridSlot, properties {horizontalAlignment=HAlign_Center, layer=2, row=1, verticalAlignment=VAlign_Center, padding=<ltrb 3 3 3 3>}}
  }
  ScaleBox_1 {
   widgetClassPath=&ScaleBox
   parent=SizeBox_0
   widget=ScaleBox_1
   slot {class=SizeBoxSlot, properties {horizontalAlignment=HAlign_Center}}
  }
  Text_Index {
   widgetClassPath=&CommonTextBlock
   bIsVariable=true
   parent=ScaleBox_1
   widget=Text_Index
   properties {
    scrollStyle=&TextScrollStyle-Base_C
    text=2
    font {
     bForceMonospaced=false
     bMaterialIsStencil=false
     fontMaterial=null
     fontObject=&HeadingNow
     letterSpacing=0
     monospacedWidth=1
     size=18
     skewAmount=0
     typefaceFontName=75 Medium
     outlineSettings {
      bApplyOutlineToDropShadows=false
      bMiteredCorners=false
      bSeparateFillAlpha=false
      outlineMaterial=null
      outlineSize=0
      outlineColor=<rgba 0 0 0 1>
     }
    }
    shadowOffset=<0 0>
   }
   slot {class=ScaleBoxSlot}
  }
  WBP_UIKit_Block_Outline {
   widgetClassPath=/FortUILibrary/UIKit/Buttons/Blocks/WBP_UIKit_Block_Outline
   bIsVariable=true
   parent=GridPanel_0
   widget=WBP_UIKit_Block_Outline
   properties {bottomLeftCornerRadius=0.15, bottomRightCornerRadius=0.15, topLeftCornerRadius=0.15, topRightCornerRadius=0.15}
   slot {class=GridSlot, properties {padding=<ltrb 0 0 0 14>}}
  }
  InputActionWidget {widgetClassPath=&CommonActionWidget, bInherited=true, widget=null}
 ]
}
