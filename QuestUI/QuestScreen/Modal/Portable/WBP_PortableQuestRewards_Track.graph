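# Horizontal reward track widget for the portable quest modal, showing staged quest rewards with progress.
# SetRewards creates a QuestUIGoalRewardVM and entry per quest, filling completed stages and giving the first incomplete one progress relative to the previous threshold.
# Arrow buttons step the punch card viewmodel's clamped SelectedIndex; ScrollToSelectedIndex centers that entry and refreshes the arrows.
# Arrows appear only for touch input, collapsing at either end, and update on input-method changes.
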WidgetBlueprint /QuestUI/QuestScreen/Modal/Portable/WBP_PortableQuestRewards_Track
graphs [
 EventGraph {nodes [
   "On Clicked (MobileButtonLeft)"!() -> (OutputDelegate, then, Button:Common Button Base Object Reference) {is=title K2Node_ComponentBoundEvent}
   n1 = Get SelectedIndex!(self:Quest UIHorizontal Punch Card Object Reference=Get QuestUIHorizontalPunchCard() -> (QuestUIHorizontalPunchCard:Quest UIHorizontal Punch Card Object Reference)) -> (SelectedIndex:int=0)
   > Set with Broadcast SelectedIndex!(
    execute
    SelectedIndex:int="Max (Integer)"(A=int - int(A=@n1), B=0)
    self:Quest UIHorizontal Punch Card Object Reference=Get QuestUIHorizontalPunchCard()
   ) -> (then, Output_Get:int=0)
   "On Clicked (MobileButtonRight)"!() -> (OutputDelegate, then, Button:Common Button Base Object Reference) {is=title K2Node_ComponentBoundEvent}
   n2 = Get OrderedContentVMs!(self:Quest UI Punch Card View Model Object Reference=Get PunchCardVM!(self:Quest UIHorizontal Punch Card Object Reference=Get QuestUIHorizontalPunchCard()) -> (PunchCardVM:Quest UI Punch Card View Model Object Reference)) -> (OrderedContentVMs:Array of Quest UI Base Content View Model Object References)
   n3 = Get SelectedIndex(self=Get QuestUIHorizontalPunchCard()) -> (SelectedIndex=0)
   > Set with Broadcast SelectedIndex(
    SelectedIndex="Min (Integer)"(A=int + int(A=@n3), B=Last Index!(TargetArray:Array of Quest UI Base Content View Model Object References=@n2) -> (ReturnValue=0)) -> (ReturnValue=0)
    self=Get QuestUIHorizontalPunchCard()
   ) -> (Output_Get=0)
   n4 = OnInputMethodChanged!() -> (OutputDelegate, then, bNewInputType:ECommonInputType Enum) {is=title K2Node_CustomEvent}
   > UpdateArrowButtons()
   Event Destruct()
   > Unbind Event from On Input Method Changed(self=Get CommonInputSubsystem(), Delegate=@n4)
   Event Construct()
   > Bind Event to On Input Method Changed(self=Get CommonInputSubsystem(), Delegate=@n4)
   > UpdateArrowButtons()
   > Set with Broadcast IsUsingCompactUI!(execute, IsUsingCompactUI:bool=IsUsingCompactUI()) -> (then, Output_Get:bool)
  ]}
 GetCurrentStageText {nodes [
   n1 = GetCurrentStageText!() -> (then, InSelectedIndex:int, InGoalRewards:Array of Quest UI Base View Model Object References) {is=title K2Node_FunctionEntry}
   > Return Node!(
    execute
    Return Value:Text=Format Text!(
     Format="LOCTEXT(\"Stage {index} of {total}\")"
     index:int=int + int(A=@n1)
     total:int=Length(TargetArray:Array of Quest UI Base View Model Object References=@n1.InGoalRewards)
    ) -> (Result)
   )
  ]}
 GetWidgetAtIndex {nodes [
   GetWidgetAtIndex!() -> (then, InIndex:int) {is=title K2Node_FunctionEntry}
   n1 = GetAllEntries!(self:Dynamic Entry Box Base Object Reference=Get EntryBox() -> (EntryBox:Dynamic Entry Box Object Reference)) -> (ReturnValue:Array of User Widget Object References)
   n2 = Get InIndex() -> (InIndex:int=0)
   > n3 = Branch(Condition=Is Valid Index!(TargetArray:Array of User Widget Object References=@n1, IndexToTest:int=@n2) -> (ReturnValue:bool))
   > Return Node!(
    execute
    Output:User Widget Object Reference="Get (a copy)"(Array:Array of User Widget Object References=@n1, Dimension 1=@n2) -> (Output:User Widget Object Reference)
   )
   Return Node(execute=@n3.else)
  ]}
 ScrollToSelectedIndex {nodes [
   n1 = ScrollToSelectedIndex!() -> (then, InIndex:int) {is=title K2Node_FunctionEntry}
   n2 = GetWidgetAtIndex!(self, InIndex:int=@n1) -> (Output:User Widget Object Reference) {is=title K2Node_CallFunction}
   > Is Valid(InputObject=@n2)
   > ScrollWidgetIntoView!(
    execute
    self:Scroll Box Object Reference=Get ScrollBox_Content() -> (ScrollBox_Content:Scroll Box Object Reference)
    WidgetToFind:Widget Object Reference=@n2
    AnimateScroll:bool=true
    ScrollDestination:EDescendantScrollDestination Enum=Center
    Padding:float=0
   ) -> (then)
   > UpdateArrowButtons()
  ]}
 SetRewards {nodes [
   SetRewards!() -> (then, InQuestVMs:Array of Quest UI Base Content View Model Object References) {is=title K2Node_FunctionEntry}
   > Reset!(execute, self:Dynamic Entry Box Object Reference=Get EntryBox(), bDeleteWidgets:bool=false) -> (then)
   > Set with Broadcast SelectedIndex(SelectedIndex=0, self=Get QuestUIHorizontalPunchCard()) -> (Output_Get=0)
   > n1 = For Each Loop(Array:Array of Quest UI Base Content View Model Object References=Get InQuestVMs() -> (InQuestVMs:Array of Quest UI Base Content View Model Object References)) -> (Array Element:Quest UI Base Content View Model Object Reference)
   > n2 = Sequence!(execute) -> (then_0, then_1, then_2:Exec)
   > n3 = Cast To QuestUIQuestVM!(execute, Object=@n1) -> (then, CastFailed, AsQuest UI Quest View Model:Quest UI Quest View Model Object Reference) {is=title K2Node_DynamicCast}
   > n4 = Set QuestVM(QuestVM:Quest UI Quest View Model Object Reference=@n3) -> (Output_Get:Quest UI Quest View Model Object Reference)
   > n5 = Construct Object from Class!(
    execute
    Class=/QuestUI/QuestScreen/Modal/Goals/QuestUIGoalRewardVM.QuestUIGoalRewardVM_C
    self
    Index:int=@n1.Array Index
    QuestVM:Quest UI Quest View Model Object Reference=@n4
    ShowProgress:bool=True
    IsLastIndex:bool=False
   ) -> (then, ReturnValue:Quest UIGoal Reward VM Object Reference)
   > Set GoalRewardVM(GoalRewardVM:Quest UIGoal Reward VM Object Reference=@n5) -> (Output_Get:Quest UIGoal Reward VM Object Reference)
   n6 = Get QuestVM() -> (QuestVM:Quest UI Quest View Model Object Reference)
   n7 = Branch(
    execute=@n2.then_1
    Condition="integer >= integer"(
     A=Get AchievedCount!(self:Quest UI Quest View Model Object Reference=@n6) -> (AchievedCount:int=0)
     B=Get RequiredCount!(self:Quest UI Quest View Model Object Reference=@n6) -> (RequiredCount:int=0)
    )
   )
   > Set with Broadcast ProgressValue!(
    execute
    ProgressValue:double=1
    self:Quest UIGoal Reward VM Object Reference=Get GoalRewardVM() -> (GoalRewardVM:Quest UIGoal Reward VM Object Reference)
   ) -> (then, Output_Get:double)
   > Set PreviousRequiredCount(PreviousRequiredCount:int=Get RequiredCount(self=Get QuestVM()) -> (RequiredCount=0)) -> (Output_Get:int=0)
   Do Once(execute=@n7.else)
   n8 = Get PreviousRequiredCount() -> (PreviousRequiredCount:int=0)
   n9 = Get QuestVM()
   n10 = int - int(A=Get AchievedCount(self=@n9) -> (AchievedCount=0), B=@n8)
   n11 = int - int(A=Get RequiredCount(self=@n9) -> (RequiredCount=0), B=@n8)
   > Set with Broadcast ProgressValue(ProgressValue=SafeDivide(A="To Float (Integer)"(InInt=@n10), B="To Float (Integer)"(InInt=@n11)), self=Get GoalRewardVM())
   n12 = Create Entry of Class!(
    execute=@n2.then_2
    self:Dynamic Entry Box Object Reference=Get EntryBox()
    EntryClass:User Widget Class Reference=&/WBP_PortableQuestRewards_Entry.WBP_PortableQuestRewards_Entry_C
   ) -> (then, ReturnValue:WBP Portable Quest Rewards Entry Object Reference)
   > SetViewModelByClass!(execute, Widget:User Widget Object Reference=@n12, ViewModel:Notify Field Value Changed Interface=Get GoalRewardVM()) -> (then)
   > SetViewModelByClass(Widget=@n12, ViewModel=Get QuestUIHorizontalPunchCard())
  ]}
 UpdateArrowButtons {nodes [
   UpdateArrowButtons() {is=title K2Node_FunctionEntry}
   > n1 = Switch on ECommonInputType!(execute, Selection:ECommonInputType Enum=GetCurrentInputType(self=Get CommonInputSubsystem())) -> (MouseAndKeyboard:Exec, Gamepad:Exec, Touch:Exec, Count:Exec) {is=title K2Node_SwitchEnum}
   n2 = Get!(execute:Exec=@n1.Touch) -> (then:Exec, else:Exec, QuestUIHorizontalPunchCard:Quest UIHorizontal Punch Card Object Reference)
   n3 = Get QuestUIHorizontalPunchCard()
   n4 = Last Index(TargetArray=Get OrderedContentVMs(self=Get PunchCardVM(self=@n3))) -> (ReturnValue=0)
   n5 = Get SelectedIndex(self=@n3) -> (SelectedIndex=0)
   > Set Visibility(
    Visibility=Select(Option 0=SelfHitTestInvisible, Option 1=Collapsed, Index="Equal (Integer)"(A=@n4, B=@n5))
    self=Get MobileButtonRight() -> (MobileButtonRight:WBP UIKit Button Generic Object Reference)
   )
   n6 = Get SelectedIndex(self=Get QuestUIHorizontalPunchCard()) -> (SelectedIndex=0)
   > Set Visibility(
    Visibility=Select(Option 0=SelfHitTestInvisible, Option 1=Collapsed, Index="Equal (Integer)"(A=@n6))
    self=Get MobileButtonLeft() -> (MobileButtonLeft:WBP UIKit Button Generic Object Reference)
   )
   SetVisibility(
    execute=[@n2.else, @n1.Count, @n1.Gamepad, @n1.MouseAndKeyboard]
    self=[Get MobileButtonRight(), Get MobileButtonLeft()]
    InVisibility=Collapsed
   )
  ]}
]
variables [
 IsUsingCompactUI {type=boolean}
 Padding_CompactUI_Content {type=Margin, default=<ltrb 96 0 96 0>}
]
widgets {
 info {inheritedWidgetCount=0, parentClass=&UserWidget, rootWidgetClass=&Border, widgetCount=10}
 widgets [
  Border_24 {
   widgetClassPath=&Border
   widget=Border_24
   properties {
    background {
     bIsDynamicallyLoaded=false
     drawAs=RoundedBox
     imageType=NoImage
     mirroring=NoMirror
     resourceName=None
     resourceObject=null
     tiling=NoTile
     imageSize=<32 32>
     margin=<ltrb 0 0 0 0>
     outlineSettings {
      bUseBrushTransparency=false
      roundingType=FixedRadius
      width=0
      color {colorUseRule=UseColor_Specified, specifiedColor=<rgba 0 0 0 0>}
      cornerRadii=<8 8 8 8>
     }
     tintColor {colorUseRule=UseColor_Specified, specifiedColor=<rgba 1 1 1 1>}
     uVRegion {bIsValid=false, max=<0 0>, min=<0 0>}
    }
    brushColor=<rgba 1 1 1 0.25>
    padding=<ltrb 24 16 24 16>
   }
  }
  VerticalBox_0 {
   widgetClassPath=&VerticalBox
   parent=Border_24
   widget=VerticalBox_0
   properties {bOverride_Cursor=true}
   slot {class=BorderSlot, properties {padding=<ltrb 24 16 24 16>}}
  }
  Text_Header {
   widgetClassPath=&CommonTextBlock
   parent=VerticalBox_0
   widget=Text_Header
   properties {
    autoWrapText=true
    mobileFontSizeMultiplier=1.1
    text="[ph] Blitz Royale Set"
    font {
     bForceMonospaced=false
     bMaterialIsStencil=false
     fontMaterial=null
     fontObject=&HeadingNow
     letterSpacing=0
     monospacedWidth=1
     size=22.5
     skewAmount=0
     typefaceFontName=75 Medium Italic
     outlineSettings {
      bApplyOutlineToDropShadows=false
      bMiteredCorners=false
      bSeparateFillAlpha=false
      outlineMaterial=null
      outlineSize=0
      outlineColor=<rgba 0 0 0 1>
     }
    }
    shadowOffset=<0 0>
   }
   slot {class=VerticalBoxSlot}
  }
  Text_Subheader {
   widgetClassPath=&CommonTextBlock
   parent=VerticalBox_0
   widget=Text_Subheader
   properties {
    mobileFontSizeMultiplier=1.15
    text="[ph] Stage 4 of 5"
    colorAndOpacity {colorUseRule=UseColor_Specified, specifiedColor=<rgba 0.5776 0.5776 0.5776 1>}
    font {
     bForceMonospaced=false
     bMaterialIsStencil=false
     fontMaterial=null
     fontObject=&HeadingNow
     letterSpacing=0
     monospacedWidth=1
     size=15.75
     skewAmount=0
     typefaceFontName=74 Regular
     outlineSettings {
      bApplyOutlineToDropShadows=false
      bMiteredCorners=false
      bSeparateFillAlpha=false
      outlineMaterial=null
      outlineSize=0
      outlineColor=<rgba 0 0 0 1>
     }
    }
    shadowOffset=<0 0>
   }
   slot {class=VerticalBoxSlot}
  }
  Overlay_0 {
   widgetClassPath=&Overlay
   parent=VerticalBox_0
   widget=Overlay_0
   slot {class=VerticalBoxSlot, properties {padding=<ltrb 0 24 0 0>}}
  }
  PostBufferUpdate {
   widgetClassPath=/Script/UMG.PostBufferUpdate
   parent=Overlay_0
   widget=PostBufferUpdate
   properties {visibility=HitTestInvisible, updateBufferInfos [{bufferToUpdate=ESlatePostRT_1, postParamUpdater=null}]}
   slot {class=OverlaySlot, properties {horizontalAlignment=HAlign_Fill, verticalAlignment=VAlign_Fill}}
  }
  ScrollBox_Content {
   widgetClassPath=&ScrollBox
   bIsVariable=true
   parent=Overlay_0
   widget=ScrollBox_Content
   properties {
    bEnableTouchScrolling=false
    navigationDestination=Center
    orientation=Orient_Horizontal
    scrollBarVisibility=Collapsed
    widgetStyle {
     barThickness=9
     bottomShadowBrush {
      bIsDynamicallyLoaded=false
      drawAs=Box
      imageType=FullColor
      mirroring=NoMirror
      resourceName=../../../Engine/Content/Slate/Common/ScrollBoxShadowBottom.png
      resourceObject=null
      tiling=NoTile
      imageSize=<16 8>
      margin=<ltrb 0.5 0 0.5 1>
      outlineSettings {
       bUseBrushTransparency=false
       roundingType=HalfHeightRadius
       width=0
       color {colorUseRule=UseColor_Specified, specifiedColor=<rgba 0 0 0 0>}
       cornerRadii=<0 0 0 0>
      }
      tintColor {colorUseRule=UseColor_Specified, specifiedColor=<rgba 1 1 1 1>}
      uVRegion {bIsValid=false, max=<0 0>, min=<0 0>}
     }
     horizontalScrolledContentPadding=<ltrb 0 0 1 0>
     leftShadowBrush {
      bIsDynamicallyLoaded=false
      drawAs=Box
      imageType=FullColor
      mirroring=NoMirror
      resourceName=None
      resourceObject=/FortUILibrary/Systems/Items/MaterialInstances/MI_UI_ListFade_Left
      tiling=NoTile
      imageSize=<8 16>
      margin=<ltrb 1 0.5 0 0.5>
      outlineSettings {
       bUseBrushTransparency=false
       roundingType=HalfHeightRadius
       width=0
       color {colorUseRule=UseColor_Specified, specifiedColor=<rgba 0 0 0 0>}
       cornerRadii=<0 0 0 0>
      }
      tintColor {colorUseRule=UseColor_Specified, specifiedColor=<rgba 1 1 1 1>}
      uVRegion {bIsValid=false, max=<0 0>, min=<0 0>}
     }
     rightShadowBrush {
      bIsDynamicallyLoaded=false
      drawAs=Box
      imageType=FullColor
      mirroring=NoMirror
      resourceName=None
      resourceObject=/FortUILibrary/Systems/Items/MaterialInstances/MI_UI_ListFade_Right
      tiling=NoTile
      imageSize=<8 16>
      margin=<ltrb 0 0.5 1 0.5>
      outlineSettings {
       bUseBrushTransparency=false
       roundingType=HalfHeightRadius
       width=0
       color {colorUseRule=UseColor_Specified, specifiedColor=<rgba 0 0 0 0>}
       cornerRadii=<0 0 0 0>
      }
      tintColor {colorUseRule=UseColor_Specified, specifiedColor=<rgba 1 1 1 1>}
      uVRegion {bIsValid=false, max=<0 0>, min=<0 0>}
     }
     topShadowBrush {
      bIsDynamicallyLoaded=false
      drawAs=Box
      imageType=FullColor
      mirroring=NoMirror
      resourceName=../../../Engine/Content/Slate/Common/ScrollBoxShadowTop.png
      resourceObject=null
      tiling=NoTile
      imageSize=<16 8>
      margin=<ltrb 0.5 1 0.5 0>
      outlineSettings {
       bUseBrushTransparency=false
       roundingType=HalfHeightRadius
       width=0
       color {colorUseRule=UseColor_Specified, specifiedColor=<rgba 0 0 0 0>}
       cornerRadii=<0 0 0 0>
      }
      tintColor {colorUseRule=UseColor_Specified, specifiedColor=<rgba 1 1 1 1>}
      uVRegion {bIsValid=false, max=<0 0>, min=<0 0>}
     }
     verticalScrolledContentPadding=<ltrb 0 0 0 1>
    }
   }
   slot {class=OverlaySlot, properties {horizontalAlignment=HAlign_Center, verticalAlignment=VAlign_Fill}}
  }
  EntryBox {
   widgetClassPath=&DynamicEntryBox
   bIsVariable=true
   parent=ScrollBox_Content
   widget=EntryBox
   properties {
    entryWidgetClass=&WBP_PortableQuestRewards_Entry_C
    numDesignerPreviewEntries=10
    entrySpacing=<32 0>
   }
   slot {class=ScrollBoxSlot, properties {padding=<ltrb 16 16 16 0>}}
  }
  MobileButtonLeft {
   widgetClassPath=&WBP_UIKit_Button_Generic
   bIsVariable=true
   parent=Overlay_0
   widget=MobileButtonLeft
   properties {
    bIsFocusable=false
    backgroundMaterial=/FortUILibrary/UIKit/Buttons/MaterialInstances/CTAButtons/MI_UIKit_BgRegular
    iconMaterial=/FortUILibrary/Foundation/Buttons/Materials/Arrows/MI_UIKit_Icon_Button_Previous
    minHeight=48
    minHeightOverride=48
    minWidth=80
    minWidthOverride=48
    mobileMinHeight=88
    mobileMinWidth=88
    outlineMaterial=/FortUILibrary/UIKit/Buttons/MaterialInstances/CTAButtons/MI_UIKit_ButtonOutline
    promptHorizontalAlignment=HAlign_Center
    useTextBlock=false
    buttonPadding=<ltrb 0 0 0 0>
    iconSize=<48 48>
    promptNudge=<0 0>
   }
   slot {class=OverlaySlot, properties {verticalAlignment=VAlign_Center, padding=<ltrb 0 0 0 28>}}
  }
  MobileButtonRight {
   widgetClassPath=&WBP_UIKit_Button_Generic
   bIsVariable=true
   parent=Overlay_0
   widget=MobileButtonRight
   properties {
    bIsFocusable=false
    backgroundMaterial=&MI_UIKit_BgRegular
    iconMaterial=/FortUILibrary/Foundation/Buttons/Materials/Arrows/MI_UIKit_Icon_Button_Next
    minHeight=48
    minHeightOverride=48
    minWidth=80
    minWidthOverride=48
    mobileMinHeight=88
    mobileMinWidth=88
    outlineMaterial=&MI_UIKit_ButtonOutline
    promptHorizontalAlignment=HAlign_Center
    useTextBlock=false
    buttonPadding=<ltrb 0 0 0 0>
    iconSize=<48 48>
    promptNudge=<0 0>
   }
   slot {class=OverlaySlot, properties {horizontalAlignment=HAlign_Right, verticalAlignment=VAlign_Center, padding=<ltrb 0 0 0 28>}}
  }
 ]
}
