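# Activatable dialog modal that shows a Battle Pass item's quest rewards as a track and preview, with a fallback message when they are unavailable in the current experience.
# Receiving a punch card viewmodel switches to the content and shares a horizontal punch card wrapper with track and preview; pips drive its selected index.
# Dialog buttons open the reward's quest category.
# Gamepad focus targets the selected track entry, and compact UI adjusts the grid.
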WidgetBlueprint /QuestUI/QuestScreen/Modal/Portable/WBP_PortableQuestRewards_Modal
graphs [
 BP_GetDesiredFocusTarget {nodes [
   Get Desired Focus Target()
   n1 = GetWidgetAtIndex!(
    self:WBP Portable Quest Rewards Track Object Reference=Get RewardTrack() -> (RewardTrack:WBP Portable Quest Rewards Track Object Reference)
    InIndex:int=Get SelectedIndex!(self:Quest UIHorizontal Punch Card Object Reference=Get HorizontalPunchCardVM() -> (HorizontalPunchCardVM:Quest UIHorizontal Punch Card Object Reference)) -> (SelectedIndex:int=0)
   ) -> (Output:User Widget Object Reference)
   > n2 = Is Valid(InputObject=@n1)
   n3 = Get WBP_UIKit_Dialog_Internal() -> (WBP_UIKit_Dialog_Internal:WBP UIKit Dialog Internal Object Reference)
   Return Node!(
    execute=@n2.Is Not Valid
    ReturnValue:Widget Object Reference="Get (a copy)"(Array:Array of WBP UIKit Button CTA Base Object References=Get ButtonWidgets!(self:WBP UIKit Dialog Internal Object Reference=@n3) -> (ButtonWidgets:Array of WBP UIKit Button CTA Base Object References)) -> (Output:WBP UIKit Button CTA Base Object Reference)
   )
   > Return Node(ReturnValue=@n1)
  ]}
 EventGraph {nodes [
   "On Clicked (Button_Back)"!() -> (OutputDelegate, then, Button:Common Button Base Object Reference) {is=title K2Node_ComponentBoundEvent}
   > DeactivateWidget()
   n1 = "On Active Pip Changed (WBP_UIKit_Pips)"!() -> (OutputDelegate, then, NewIndex:int=0) {is=title K2Node_ComponentBoundEvent}
   > Set with Broadcast SelectedIndex!(execute, SelectedIndex:int=@n1.NewIndex, self:Quest UIHorizontal Punch Card Object Reference=Get HorizontalPunchCardVM()) -> (then, Output_Get:int=0)
   n2 = ButtonClicked!() -> (OutputDelegate, then, ButtonIndex:int) {is=title K2Node_CustomEvent}
   n3 = GetReward!(self:Season Pass Reward VM Object Reference=Get SeasonPassRewardVM() -> (SeasonPassRewardVM:Season Pass Reward VM Object Reference)) -> (ReturnValue:Athena Season Item Entry Base Object Reference)
   > NavigateToQuestScreen!(
    execute
    Context:Widget Object Reference=Self-Reference()
    StartingCategoryTag:"Gameplay Tag Structure (by ref)"="Get (a ref)"(
     Array:Array of Gameplay Tag Structures=GetQuestCategoryTags!(Context:Widget Object Reference, RewardEntry:Athena Season Item Entry Base Object Reference=@n3) -> (ReturnValue_GameplayTags:Array of Gameplay Tag Structures)
     Dimension 1=0
    ) -> (Output:"Gameplay Tag Structure (by ref)")
   ) -> (then)
   n4 = OnInputMethodChanged!() -> (OutputDelegate, then, bNewInputType:ECommonInputType Enum) {is=title K2Node_CustomEvent}
   > n5 = Switch on ECommonInputType!(execute, Selection:ECommonInputType Enum=@n4.bNewInputType) -> (MouseAndKeyboard:Exec, Gamepad:Exec, Touch:Exec, Count:Exec) {is=title K2Node_SwitchEnum}
   RequestRefreshFocus(execute=@n5.Gamepad, self:Common Activatable Widget Object Reference)
   Event Destruct()
   > Unbind Event from On Input Method Changed(self=Get CommonInputSubsystem(), Delegate=@n4)
   Event On Activated()
   > Bind Event to Dialog Button Clicked Dispatcher(self:WBP UIKit Dialog Internal Object Reference=Get WBP_UIKit_Dialog_Internal(), Delegate=@n2) {is=title K2Node_AddDelegate}
   Event On Deactivated()
   > Unbind Event from Dialog Button Clicked Dispatcher(self:WBP UIKit Dialog Internal Object Reference=Get WBP_UIKit_Dialog_Internal(), Delegate=@n2) {is=title K2Node_RemoveDelegate}
   Event Construct()
   > n6 = Construct Object from Class(Class=&/DialogVM_PortableQuestRewards.DialogVM_PortableQuestRewards_C) -> (ReturnValue:Dialog VM Portable Quest Rewards Object Reference)
   > SetDialogData!(execute, self:WBP UIKit Dialog Internal Object Reference=Get WBP_UIKit_Dialog_Internal(), Data:UI Kit Dialog Viewmodel Object Reference=@n6) -> (then)
   > Bind Event to On Input Method Changed(self=Get CommonInputSubsystem(), Delegate=@n4)
   > SetMobileStyle() {is=title K2Node_CallFunction}
  ]}
 GetArrayLength {nodes [
   n1 = GetArrayLength!() -> (then, InArray:Array of Quest UI Base Content View Model Object References) {is=title K2Node_FunctionEntry}
   > Return Node!(execute, Length:int=Length(TargetArray:Array of Quest UI Base Content View Model Object References=@n1))
  ]}
 NavRuleDesiredFocusTarget {nodes [
   NavRuleDesiredFocusTarget!() -> (then, Navigation:EUINavigation Enum) {is=title K2Node_FunctionEntry}
   > Return Node!(
    execute
    ReturnValue:Widget Object Reference=GetDesiredFocusTarget!(self:Common Activatable Widget Object Reference) -> (ReturnValue:Widget Object Reference)
   )
  ]}
 SetHorizontalPunchcardVM {nodes [
   n1 = SetHorizontalPunchcardVM!() -> (then, InPunchCardVM:Quest UI Punch Card View Model Object Reference) {is=title K2Node_FunctionEntry}
   > SetActiveWidgetIndex(self:Common Visibility Switcher Object Reference=Get Switcher_Root() -> (Switcher_Root:Common Visibility Switcher Object Reference), Index=1)
   > n2 = Construct Object from Class!(
    execute
    Class=/QuestUI/QuestScreen/Modal/Goals/QuestUIHorizontalPunchCard.QuestUIHorizontalPunchCard_C
    self
    PunchCardVM:Quest UI Punch Card View Model Object Reference=@n1
   ) -> (then, ReturnValue:Quest UIHorizontal Punch Card Object Reference)
   > n3 = Set with Broadcast HorizontalPunchCardVM!(execute, HorizontalPunchCardVM:Quest UIHorizontal Punch Card Object Reference=@n2) -> (then, Output_Get:Quest UIHorizontal Punch Card Object Reference)
   > n4 = Set QuestUIHorizontalPunchCard!(
    execute
    QuestUIHorizontalPunchCard:Quest UIHorizontal Punch Card Object Reference=@n3
    self:WBP Portable Quest Rewards Track Object Reference=Get RewardTrack()
   ) -> (then, Output_Get:Quest UIHorizontal Punch Card Object Reference)
   > Set QuestUIHorizontalPunchCard(
    QuestUIHorizontalPunchCard=@n4
    self:WBP Portable Quest Rewards Preview Object Reference=Get RewardPreview() -> (RewardPreview:WBP Portable Quest Rewards Preview Object Reference)
   )
  ]}
 SetMobileStyle {nodes [
   SetMobileStyle() {is=title K2Node_FunctionEntry}
   > Branch(Condition=IsUsingCompactUI())
   n1 = Get GP_Content() -> (GP_Content:Grid Panel Object Reference)
   > SetColumnFill!(execute, self:Grid Panel Object Reference=@n1, ColumnIndex:int=0, Coefficient:float=0.7) -> (then)
   > SetColumnFill(self=@n1, ColumnIndex=1, Coefficient=0.3)
   > SetGridSlotPadding!(
    execute
    Widget:Widget Object Reference=Get SZ_Preview() -> (SZ_Preview:Size Box Object Reference)
    Padding_Left:float=40
    Padding_Top:float=40
    Padding_Right:float=0
    Padding_Bottom:float=-80
   ) -> (then)
  ]}
 SetPunchCardFromSeasonPassReward {nodes [
   n1 = SetPunchCardFromSeasonPassReward!() -> (then, InSeasonPassRewardVM:Season Pass Reward VM Object Reference) {is=title K2Node_FunctionEntry}
   > Is Valid(InputObject=@n1)
   > n2 = Cast To AthenaSeasonItemEntryQuest!(execute, Object=GetReward(self=@n1)) -> (then, CastFailed, AsQuest:Quest Object Reference) {is=title K2Node_DynamicCast}
   > n3 = CreateQuestUIPunchCardVM!(execute, TemplateID:String=GetQuestBundleTemplateId!(self:Quest Object Reference=@n2) -> (ReturnValue:String)) -> (then, ReturnValue:Quest UI Punch Card View Model Object Reference)
   > Set QuestUIPunchCardVM(QuestUIPunchCardVM:Quest UI Punch Card View Model Object Reference=@n3) -> (Output_Get:Quest UI Punch Card View Model Object Reference)
  ]}
]
variables [
 HorizontalPunchCardVM {type=QuestUIHorizontalPunchCard_C, default=null}
 IsUsingCompactUI {type=boolean}
]
widgets {
 info {
  inheritedWidgetCount=0
  namedSlotCount=4
  parentClass=&CommonActivatableWidget
  rootWidgetClass=/FortUILibrary/Systems/Dialog/WBP_UIKit_Dialog_Internal.WBP_UIKit_Dialog_Internal_C
  widgetCount=13
 }
 widgets [
  WBP_UIKit_Dialog_Internal {
   widgetClassPath=/FortUILibrary/Systems/Dialog/WBP_UIKit_Dialog_Internal
   bIsVariable=true
   widget=WBP_UIKit_Dialog_Internal
   properties {
    designTimePreview_DialogVM=&DialogVM_PortableQuestRewards_C
    useBackground=true
    useBackplate=false
    backgroundProperties {
     material=&/MI_UI_PortableQuestRewards_ModalBackground
     image Size=<32 32>
     placement {
      horizontal Alignment=HAlign_Fill
      vertical Alignment=VAlign_Fill
      grid {column=0, column Span=2, layer=1, row=0, row Span=0, nudge=<0 0>}
      padding=<ltrb 0 0 0 0>
     }
    }
    buttons {
     each Button Horizontal Fill=false
     each Button Min Width=397
     each Button Size=Medium
     focuse Button=0
     space Between Buttons=30
     cTA Buttons Types [Regular]
     group Placement {
      horizontal Alignment=HAlign_Fill
      vertical Alignment=VAlign_Bottom
      grid {column=1, column Span=0, layer=0, row=2, row Span=0, nudge=<0 0>}
      padding=<ltrb 80 64 80 48>
     }
    }
    dialogPlacement {
     horizontal Alignment=HAlign_Center
     vertical Alignment=VAlign_Center
     grid {column=0, column Span=0, layer=1, row=0, row Span=0, nudge=<0 0>}
     padding=<ltrb 160 0 160 20>
    }
    dialogSIze {
     height Override {enable=true, value=900}
     max Aspect Ratio {enable=false, value=0}
     max Desired Height {enable=false, value=0}
     max Desired Width {enable=true, value=2000}
     min Aspect Ratio {enable=false, value=0}
     min Desired Height {enable=false, value=0}
     min Desired Width {enable=true, value=1600}
     width Override {enable=false, value=0}
    }
   }
  }
  Switcher_Root {widgetClassPath=&CommonVisibilitySwitcher, bIsVariable=true, namedSlotHost=WBP_UIKit_Dialog_Internal, widget=Switcher_Root}
  Overlay_43 {widgetClassPath=&Overlay, parent=Switcher_Root, widget=Overlay_43, slot {class=CommonVisibilitySwitcherSlot}}
  CommonTextBlock_86 {
   widgetClassPath=&CommonTextBlock
   parent=Overlay_43
   widget=CommonTextBlock_86
   properties {
    autoWrapText=true
    text="Quest rewards for this item aren’t available in this Fortnite experience. Return to the lobby to view this item in the Battle Pass, or play the correct Fortnite experience for this quest reward."
    font {
     bForceMonospaced=false
     bMaterialIsStencil=false
     fontMaterial=null
     fontObject=&HeadingNow
     letterSpacing=0
     monospacedWidth=1
     size=22.5
     skewAmount=0
     typefaceFontName=74 Regular
     outlineSettings {
      bApplyOutlineToDropShadows=false
      bMiteredCorners=false
      bSeparateFillAlpha=false
      outlineMaterial=null
      outlineSize=0
      outlineColor=<rgba 0 0 0 1>
     }
    }
    shadowOffset=<0 0>
   }
   slot {class=OverlaySlot, properties {horizontalAlignment=HAlign_Center, verticalAlignment=VAlign_Center, padding=<ltrb 200 0 200 0>}}
  }
  GP_Content {
   widgetClassPath=&GridPanel
   bIsVariable=true
   parent=Switcher_Root
   widget=GP_Content
   properties {bOverride_Cursor=true, columnFill [0.6,0.4], rowFill [1]}
   slot {class=CommonVisibilitySwitcherSlot, properties {padding=<ltrb 80 0 80 0>}}
  }
  RewardTrack {
   widgetClassPath=&/WBP_PortableQuestRewards_Track
   bIsVariable=true
   parent=GP_Content
   widget=RewardTrack
   properties {navigation=&self:WidgetTree.RewardTrack.WidgetNavigation_1}
   slot {class=GridSlot, properties {rowSpan=1, verticalAlignment=VAlign_Center}}
  }
  SZ_Preview {
   widgetClassPath=&SizeBox
   bIsVariable=true
   parent=GP_Content
   widget=SZ_Preview
   properties {bOverride_WidthOverride=true, widthOverride=480}
   slot {class=GridSlot, properties {column=1, rowSpan=1, padding=<ltrb 40 40 0 0>}}
  }
  RewardPreview {
   widgetClassPath=&/WBP_PortableQuestRewards_Preview
   bIsVariable=true
   parent=SZ_Preview
   widget=RewardPreview
   slot {class=SizeBoxSlot}
  }
  WBP_UIKit_Pips {widgetClassPath=/FortUILibrary/UIKit/PipList/WBP_UIKit_Pips, bIsVariable=true, namedSlotHost=RewardPreview, widget=WBP_UIKit_Pips}
  Overlay_1 {widgetClassPath=&Overlay, namedSlotHost=WBP_UIKit_Dialog_Internal, widget=Overlay_1}
  SafeZone_78 {
   widgetClassPath=&SafeZone
   bIsVariable=true
   parent=Overlay_1
   widget=SafeZone_78
   properties {padLeft=false, padTop=false}
   slot {class=OverlaySlot, properties {horizontalAlignment=HAlign_Fill, verticalAlignment=VAlign_Fill}}
  }
  WBP_ActionBarContainer {
   widgetClassPath=/Game/UI/Frontend/BottomBar/WBP_ActionBarContainer
   parent=SafeZone_78
   widget=WBP_ActionBarContainer
   slot {class=SafeZoneSlot, properties {hAlign=HAlign_Right, vAlign=VAlign_Bottom, padding=<ltrb 76 32 76 32>}}
  }
  Button_Back {
   widgetClassPath=/FortUILibrary/UIKit/Buttons/WBP_UIKit_Button_Quiet
   bIsVariable=true
   namedSlotHost=WBP_ActionBarContainer
   widget=Button_Back
   properties {text=Back, triggeringInputAction {dataTable=&InputActionDataTable, rowName=GenericBack}}
  }
 ]
}
