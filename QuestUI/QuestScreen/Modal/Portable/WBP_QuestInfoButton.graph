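# Widget Blueprint for a "View Quest" button that opens a portable quest rewards modal for a season pass reward.
# CreatePunchCardVM derives a punch card view model from a quest-type reward's bundle template ID, and Set Quest Punch Card stores it, showing the initially collapsed button only when valid.
# Clicking fires an element-interacted event, then creates the modal with the punch card and currently selected reward view models and pushes it as a modal panel.
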WidgetBlueprint /QuestUI/QuestScreen/Modal/Portable/WBP_QuestInfoButton
graphs [
 CreatePunchCardVM {nodes [
   n1 = CreatePunchCardVM!() -> (then, SeasonPassRewardVM:Season Pass Reward VM Object Reference) {is=title K2Node_FunctionEntry}
   n2 = Cast To AthenaSeasonItemEntryQuest!(Object=GetReward!(self:Season Pass Reward VM Object Reference=@n1) -> (ReturnValue:Athena Season Item Entry Base Object Reference)) -> (AsQuest:Quest Object Reference, bSuccess:bool=~) {is=title K2Node_DynamicCast}
   > n3 = Branch(Condition=@n2.bSuccess)
   > n4 = CreateQuestUIPunchCardVM!(execute, TemplateID:String=GetQuestBundleTemplateId!(self:Quest Object Reference=@n2) -> (ReturnValue:String)) -> (then, ReturnValue:Quest UI Punch Card View Model Object Reference)
   > Return Node!(execute, QuestUIPunchCardVM:Quest UI Punch Card View Model Object Reference=@n4)
   Return Node(execute=@n3.else)
  ]}
 EventGraph {nodes [
   n1 = "On Clicked (InfoButton)"!() -> (OutputDelegate, then, Button:Common Button Base Object Reference) {is=title K2Node_ComponentBoundEvent}
   > Fire Event Element Interacted(Element=@n1.Button)
   > ShowQuestModal()
  ]}
 PunchCardToVisibility {nodes [
   n1 = PunchCardToVisibility!() -> (then, QuestUIPunchCardVM:Quest UI Punch Card View Model Object Reference) {is=title K2Node_FunctionEntry}
   > Return Node!(execute, NewParam:ESlateVisibility Enum=Select(Option 0=Collapsed, Option 1=SelfHitTestInvisible, Index=IsValid(Object=@n1)))
  ]}
 Set Quest Punch Card {nodes [
   n1 = Set Quest Punch Card!() -> (then, QuestUIPunchCardVM:Quest UI Punch Card View Model Object Reference) {is=title K2Node_FunctionEntry}
   > n2 = Set QuestUIPunchCardVM(QuestUIPunchCardVM:Quest UI Punch Card View Model Object Reference=@n1) -> (Output_Get:Quest UI Punch Card View Model Object Reference)
   > SetVisibility(
    self=Get InfoButton() -> (InfoButton:WBP UIKit Button Regular Object Reference)
    InVisibility="To Visibility (Boolean)"!(
     bIsVisible:bool=IsValid(Object=@n2)
     TrueVisibility:ESlateVisibility Enum=SelfHitTestInvisible
     FalseVisibility:ESlateVisibility Enum=Collapsed
    ) -> (ReturnValue:ESlateVisibility Enum=Visible)
   )
  ]}
 ShowQuestModal {nodes [
   ShowQuestModal() {is=title K2Node_FunctionEntry}
   > n1 = Create Widget!(
    execute:Exec
    Class:User Widget Class Reference=&/WBP_PortableQuestRewards_Modal.WBP_PortableQuestRewards_Modal_C
    OwningPlayer:Player Controller Object Reference
    QuestUIPunchCardVM:Quest UI Punch Card View Model Object Reference=Get QuestUIPunchCardVM() -> (QuestUIPunchCardVM:Quest UI Punch Card View Model Object Reference)
    SeasonPassRewardVM:Season Pass Reward VM Object Reference=Get SelectedReward!(self:Season Pass Selection VM Object Reference=Get SeasonPassSelectionVM() -> (SeasonPassSelectionVM:Season Pass Selection VM Object Reference)) -> (SelectedReward:Season Pass Reward VM Object Reference)
   ) -> (then:Exec, ReturnValue:WBP Portable Quest Rewards Modal Object Reference)
   > TryPushWidgetAsModalPanel!(execute, PushWidget:User Widget Object Reference=@n1) -> (then)
  ]}
]
variables []
widgets {
 info {
  inheritedWidgetCount=0
  parentClass=&CommonActivatableWidget
  rootWidgetClass=/FortUILibrary/UIKit/Buttons/WBP_UIKit_Button_Regular.WBP_UIKit_Button_Regular_C
  widgetCount=1
 }
 widgets [InfoButton {
   widgetClassPath=&WBP_UIKit_Button_Regular
   bIsVariable=true
   widget=InfoButton
   properties {
    bHideInputActionWithKeyboard=true
    bIsFocusable=false
    hoveredSound=/Game/Sounds/Fort_UI_Sounds/FrontEnd/BattlePass/BattlePass_Button_Hover_Tile_Default_Cue
    mobileFontSizeMultiplierOverride=1.33
    overrideDefaultHeight=true
    overrideDefaultMobileFontSizeMultiplier=true
    pressedSound=/Game/Sounds/Fort_UI_Sounds/FrontEnd/BattlePass/BattlePass_Button_Click_Whoosh_Cue
    text=View Quest
    visibility=Collapsed
    triggeringInputAction {dataTable=&InputActionDataTable, rowName=GenericSpecialLeft}
   }
  }]
}
